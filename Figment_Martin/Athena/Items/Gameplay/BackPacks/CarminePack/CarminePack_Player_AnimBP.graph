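# Player animation layer Blueprint for the CarminePack backpack that adds pitch aiming to the fist beam pose.
# Each update sets IsBeam from the JimBeam_AO curve; a Carmine beam tag check is present but unused.
# ItemFinalPoseOverride crossfades between the cached input pose and a beam state driven by IsBeam.
# The beam state stacks the Jim_FistBeam_AO aim offset twice, full then partial alpha, overdriving pitch with yaw held at zero.
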AnimBlueprint /Figment_Martin/Athena/Items/Gameplay/BackPacks/CarminePack/CarminePack_Player_AnimBP
target_skeleton=&Fortnite_M_Avg_Player_Skeleton
graphs [
 AnimGraph {nodes [Output Pose()]}
 EventGraph {nodes [
   Event BlueprintUpdateAnimation()
   > Set IsBeam(IsBeam:bool="float > float"(A=GetCurveValue(CurveName=JimBeam_AO), B=0.1)) -> (Output_Get:bool)
   HasMatchingGameplayTag!(
    self:Gameplay Tag Asset Interface Interface=Get FortPlayerPawn() -> (FortPlayerPawn:Fort Player Pawn Object Reference)
    TagToCheck:Gameplay Tag Structure=(TagName="Athena.Carmine.Ability.Beam")
   ) -> (ReturnValue:bool)
   Get FortPlayerPawn()
  ]}
 ItemAimOffset {nodes [
   Input Pose!() -> (Pose, AimOffsetAlpha:double, Yaw:double, Pitch:double) {name=InPose_AimOffset}
   Input Pose!() -> (Pose) {name=InPose_PreAimOffsets}
   Output Pose() {layerGroup=Item Override}
  ]}
 ItemFallAdditive {nodes [
   Input Pose() {name=InPoseItemFallAdditive}
   Output Pose() {layerGroup=Item Override}
  ]}
 ItemFallLandAdditive {nodes [
   Input Pose() {name=InPoseItemFallLandAdditive}
   Output Pose() {layerGroup=Item Override}
  ]}
 ItemFallLevel2Override {nodes [
   Input Pose() {name=InPose_FallLevel2}
   Output Pose() {layerGroup=Item Override}
  ]}
 ItemFinalPoseOverride {nodes [
   AnimGraphNode_StateMachine_0 = Beam AO SM() {is=title AnimGraphNode_StateMachine}
   Output Pose(Result=@AnimGraphNode_StateMachine_0) {layerGroup=Item Override}
   Save cached pose 'FinalInputPose'(Pose=Input Pose() {name=InFinalPose}) {cacheName=FinalInputPose}
  ]}
 Beam AO SM {
  entry=No Beam AO
  path=ItemFinalPoseOverride.AnimGraphNode_StateMachine_0.Beam AO SM
  ids=id
  nodes [
   AnimStateNode_0 = No Beam AO() {is=title AnimStateNode}
   AnimStateNode_1 = Yes Beam AO() {is=title AnimStateNode}
   AnimStateTransitionNode_0 = AnimStateTransitionNode() {bAllowInertializationForSelfTransitions=true, crossfadeDuration=0.5}
   AnimStateTransitionNode_1 = AnimStateTransitionNode() {bAllowInertializationForSelfTransitions=true, crossfadeDuration=0.3}
  ]
  links [
   {
    dst=Yes Beam AO
    rule=ItemFinalPoseOverride.AnimGraphNode_StateMachine_0.Beam AO SM.AnimStateTransitionNode_0.Transition
    src=No Beam AO
    via=AnimStateTransitionNode_0
   }
   {
    dst=No Beam AO
    rule=ItemFinalPoseOverride.AnimGraphNode_StateMachine_0.Beam AO SM.AnimStateTransitionNode_1.Transition
    src=Yes Beam AO
    via=AnimStateTransitionNode_1
   }
  ]
 }
 No Beam AO {
  path=ItemFinalPoseOverride.AnimGraphNode_StateMachine_0.Beam AO SM.AnimStateNode_0.No Beam AO
  nodes [Output Animation Pose(Result=Use cached pose 'FinalInputPose'())]
 }
 Yes Beam AO {
  path=ItemFinalPoseOverride.AnimGraphNode_StateMachine_0.Beam AO SM.AnimStateNode_1.Yes Beam AO
  nodes [
   n1 = Use cached pose 'FinalInputPose'()
   n2 = Get!() -> (AdjustedAim_ResultingYaw:float=~, AdjustedAim_ResultingPitch:float=~, AdjustedAim_ZeroOutPitchWeightCurveName:Name)
   Output Animation Pose(Result=AimOffset Player 'Jim_FistBeam_AO'!(
     X:float=0
     Y:float=@n2.AdjustedAim_ResultingPitch
     BasePose:Pose Link Structure=AimOffset Player 'Jim_FistBeam_AO'(X=0, Y=@n2.AdjustedAim_ResultingPitch, BasePose=@n1, Alpha=1) {blendSpace=/Figment_Martin/Animation/Game/MainPlayer/Combat/Gadgets/ExtraLarge/Jim/Jim_FistBeam_AO}
     Alpha:float=0.3
    ) -> (Pose:Pose Link Structure) {blendSpace=&Jim_FistBeam_AO})
  ]
 }
 Transition {
  path=ItemFinalPoseOverride.AnimGraphNode_StateMachine_0.Beam AO SM.AnimStateTransitionNode_0.Transition
  nodes [Result(bCanEnterTransition=Get IsBeam() -> (IsBeam:bool))]
 }
 Transition {
  path=ItemFinalPoseOverride.AnimGraphNode_StateMachine_0.Beam AO SM.AnimStateTransitionNode_1.Transition
  nodes [Result(bCanEnterTransition=NOT Boolean(A=Get IsBeam()))]
 }
 ItemFullBodyAimOffsets {nodes [
   Input Pose!() -> (Pose, FullBody_AimOffsetAlpha:double, Aim_Yaw:double, Aim_Pitch:double) {name=InPoseFullBodyAimOffsets}
   Input Pose!() -> (Pose) {name=InPosePreFullBodyAimOffsets}
   Output Pose() {layerGroup=Item Override}
  ]}
 ItemFullBodyOverride {nodes [
   Input Pose() {name=InPoseFullBody}
   Output Pose() {layerGroup=Item Override}
  ]}
 ItemIdleOverride {nodes [
   Input Pose() {name=InPose_Idle}
   Output Pose() {layerGroup=Item Override}
  ]}
 ItemJetPackJumpAdditive {nodes [Output Pose() {layerGroup=Item Override}]}
 ItemJetPackStartAdditive {nodes [Output Pose() {layerGroup=Item Override}]}
 ItemJumpLoopAdditive {nodes [
   Input Pose() {name=InPoseItemJumpLoopAdditive}
   Output Pose() {layerGroup=Item Override}
  ]}
 ItemJumpOverride {nodes [
   Input Pose() {name=InPoseJump}
   Output Pose() {layerGroup=Item Override}
  ]}
 ItemJumpUpAdditive {nodes [
   Input Pose() {name=InPoseItemJumpUpAdditive}
   Output Pose() {layerGroup=Item Override}
  ]}
 ItemPreAimOffsets {nodes [
   Input Pose() {name=InPosePreAimOffsets}
   Output Pose() {layerGroup=Item Override}
  ]}
 ItemPreIK {nodes [
   Input Pose() {name=InPose_PreIK}
   Output Pose() {layerGroup=Item Override}
  ]}
 ItemRightHandIK {nodes [
   Input Pose() {name=InPosePreRightHandIK}
   Input Pose() {name=InPosePostRightHandIK}
   Output Pose() {layerGroup=Item Override}
  ]}
 ItemSkeletalControl {nodes [
   Input Pose()
   Output Pose() {layerGroup=Item Override}
  ]}
 ItemSwimJumpFallLoop {nodes [Output Pose() {layerGroup=Item Override}]}
 ItemSwimJumpSurfaceEnd {nodes [Output Pose() {layerGroup=Item Override}]}
 ItemSwimJumpSurfaceLoop {nodes [Output Pose() {layerGroup=Item Override}]}
 ItemUpperBody {nodes [
   Input Pose!() -> (Pose, InputParam:Fort Anim Input Adjusted Aim Structure) {name=InPoseUpperBody}
   Output Pose() {layerGroup=Item Override}
  ]}
 ItemZipLineStartAdditive {nodes [Output Pose() {layerGroup=Item Override}]}
]
variables [IsBeam {type=boolean}]
