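; Carmine pack deco tool that acts as the event manager for its holder's abilities and stone visuals.
; Primary and alternate fire send ability gameplay events unless blocking tags exist; primary also requires a landed flag and passes a punch-variant value as magnitude.
; Alternate fire while falling clears that flag; landing restores it and removes trail and skydive effects.
; Replicated stone tags reveal body-material elements on clients, with delayed retries until the mesh is valid.
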(event Custom|BindPartchanged
  ;[1 0.1873 0] Loop checking for skeletal mesh to be valid when entering relevancy.
  (if (not (Networking|IsServer))
    (Variables|Default|SetInt 0)
    (bind _asfort_player_pawn (Utilities|Casting|CastToFortPlayerPawn (Game|GetInstigator))
      (:then
        (Utilities|IsValid (Customization|GetSkeletalMeshforPartType _asfort_player_pawn "Body" :!class /Script/FortniteGame.FortPlayerPawn)
          (:Is%20Valid
            _)
          (:Is%20Not%20Valid
            (if (not (> (Variables|Default|GetInt) 5))
              (Utilities|FlowControl|RetriggerableDelay 2.0)
              (Math|Integer|IncrementInt (Variables|Default|GetInt)
                (:%20%20
                  _)))))
        (bind _returnvalue (Fortnite|GetDynamicMaterialInstance (Customization|GetSkeletalMeshforPartType _asfort_player_pawn "Body" :!class /Script/FortniteGame.FortPlayerPawn) 1))
        (for _array_element (Variables|Default|GetStoneTags)
          (switch Utilities|FlowControl|Switch|SwitchonGameplayTag _array_element
            (:SpecialActor.Ashton.Object.B
              (Development|PrintString "Stone" true true "(R=0.000000,G=0.264690,B=1.000000,A=1.000000)" 5.0)
              (Rendering|Material|SetScalarParameterValue _returnvalue "Hide Element 03" 0.0 :!class /Script/Engine.MaterialInstanceDynamic))
            (:SpecialActor.Ashton.Object.G
              (Development|PrintString "Stone" true true "(R=0.124658,G=1.000000,B=0.000000,A=1.000000)" 5.0)
              (Rendering|Material|SetScalarParameterValue _returnvalue "Hide Element 01" 0.0 :!class /Script/Engine.MaterialInstanceDynamic))
            (:SpecialActor.Ashton.Object.O
              (Development|PrintString "Stone" true true "(R=1.000000,G=0.195150,B=0.000000,A=1.000000)" 5.0)
              (Rendering|Material|SetScalarParameterValue _returnvalue "Hide Element 05" 0.0 :!class /Script/Engine.MaterialInstanceDynamic))
            (:SpecialActor.Ashton.Object.P
              (Development|PrintString "Stone" true true "(R=0.661715,G=0.000000,B=1.000000,A=1.000000)" 5.0)
              (Rendering|Material|SetScalarParameterValue _returnvalue "Hide Element 02" 0.0 :!class /Script/Engine.MaterialInstanceDynamic))
            (:SpecialActor.Ashton.Object.R
              (Development|PrintString "Stone" true true "(R=1.000000,G=0.000000,B=0.000000,A=1.000000)" 5.0)
              (Rendering|Material|SetScalarParameterValue _returnvalue "Hide Element 04" 0.0 :!class /Script/Engine.MaterialInstanceDynamic))
            (:SpecialActor.Ashton.Object.Y
              (Development|PrintString "Stone" true true "(R=1.000000,G=0.885502,B=0.000000,A=1.000000)" 5.0)
              (Rendering|Material|SetScalarParameterValue _returnvalue "Hide Element 06" 0.0 :!class /Script/Engine.MaterialInstanceDynamic)))))
      (:CastFailed
        _))))

(event EventBeginPlay
  ;[1 1 1] Ability handler. This deco tool acts as an event manager for the abilities running when the item is picked up.
  ;;[0.4711 0.4898 1] When you gain this item, bind your land event
  (bind _asfort_pawn (Utilities|Casting|CastToFortPawn (Game|GetInstigator)))
  (EventDispatchers|BindEventtoOnPawnLanded _asfort_pawn (delegate Custom|Landed)))

(event Custom|Landed
  (params (Hit :type Hit%20Result%20Structure%20%28by%20ref%29))
  ;[1 1 1] Ability handler. This deco tool acts as an event manager for the abilities running when the item is picked up.
  ;;[0.5339 1 0.5245] When you land, set your variables or remove other effects
  (Variables|Default|SethasLandedAfterBeam true)
  (bind _asfort_pawn (Utilities|Casting|CastToFortPawn (Game|GetInstigator)))
  (GameplayEffects|RemoveActiveEffectswithSourceTags (Ability|GetAbilitySystemComponent _asfort_pawn) (Variables|Default|GetTrailsCue))
  (GameplayEffects|RemoveActiveEffectswithSourceTags (Ability|GetAbilitySystemComponent _asfort_pawn) (Variables|Default|GetSkydiveCue)))

(event DecoTool|EventOnPressSecondaryFire
  (params (FortDecoHelper :type Fort%20Deco%20Helper%20Object%20Reference))
  (bind _returnvalue (Game|GetInstigator))
  ;[1 1 1] Ability handler. This deco tool acts as an event manager for the abilities running when the item is picked up.
  ;;[0.3882 1 0.5623] When you use your alternate fire, trigger the ability
  (bind _returnvalue_2 (GameplayTags|HasAnyMatchingGameplayTags(Message) (Game|GetInstigator) (Variables|Default|GetCarmine_Ability_Tags)))
  (if (not _returnvalue_2)
    (bind _asfort_pawn (Utilities|Casting|CastToFortPawn _returnvalue))
    (Development|PrintString "pressed alt fire" false false)
    (if (AI|Components|NavMovement|IsFalling (Variables|/Script/Engine.Character|GetCharacterMovement _asfort_pawn))
      (if (Variables|Default|GethasLandedAfterBeam)
        (Variables|Default|SethasLandedAfterBeam false))
      (else
        _))
    (Ability|SendGameplayEventtoActor _returnvalue "(TagName=\"Abilities.Generic.ActiveAbility.Activate\")" (Utilities|Struct|MakeGameplayEventData "(TagName=\"Abilities.Generic.ActiveAbility.Activate\")"))))

(event DecoTool|EventOnPressTrigger
  (params (FortDecoHelper :type Fort%20Deco%20Helper%20Object%20Reference))
  ;[1 1 1] Ability handler. This deco tool acts as an event manager for the abilities running when the item is picked up.
  ;;[0.3904 0.4224 1] When you press the primary fire send this event to your hero
  (bind _returnvalue (GameplayTags|HasAnyMatchingGameplayTags(Message) (Game|GetInstigator) (Variables|Default|GetCarmine_Ability_Tags)))
  (if (and (not _returnvalue) (Variables|Default|GethasLandedAfterBeam))
    (Math|Float|CompareFloat (Variables|Default|GetPunchPicker) 0.5
      (:>
        (Variables|Default|SetPunchPicker 0.0))
      (:==
        (Variables|Default|SetPunchPicker (+ (Variables|Default|GetPunchPicker) 0.5)))
      (:<
        _))
    (Ability|SendGameplayEventtoActor (Game|GetInstigator) "(TagName=\"Abilities.Generic.TriggeredAbility.Activate\")" (Utilities|Struct|MakeGameplayEventData :ContextHandle "()" :InstigatorTags "(GameplayTags=)" :TargetTags "(GameplayTags=)" :EventMagnitude (Variables|Default|GetPunchPicker) :TargetData "()"))))

(fn OnRep_StoneTags
  (CallFunction|BindPartchanged)
  (return))

(fn ConstructionScript)
