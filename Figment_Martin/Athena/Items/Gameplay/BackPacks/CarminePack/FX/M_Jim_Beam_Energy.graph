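# Unlit translucent mesh-particle material for an energy beam effect in the CarminePack backpack FX folder.
# Panning smoke and distortion noise, tiled by object radius, blends two HDR colors and drives opacity with end, axis-aligned fresnel, depth and camera fades.
# Lower feature levels use simpler opacity; a static switch adds time-of-day emissive scaling.
# World position offset ripples along normals using noise or a sine texture, optionally bending toward an EndPoint parameter.
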Material /Figment_Martin/Athena/Items/Gameplay/BackPacks/CarminePack/FX/M_Jim_Beam_Energy
graph {
 properties {bUsedWithMeshParticles=true, blendMode=BLEND_Translucent, pixelDepthOffsetMode=PDOM_Legacy, shadingModel=MSM_Unlit, usedShadingModels=Unlit}
 nodes [
  Constant() {r=1}
  n1 = ParticleColor()
  n2 = AppendVector(A=Constant() {r=1}, B=Divide(A=ObjectRadius(), B=ScalarParameter() {defaultValue=3500, parameterName=BeamTile}))
  n3 = TextureCoordinate() {uTiling=8, vTiling=16}
  n4 = Multiply(A=Time(), B=ScalarParameter() {defaultValue=4, parameterName=PanSpeed})
  n5 = TextureSample(UVs=Panner(Coordinate=Multiply(A=@n2, B=@n3), Time=@n4) {speedX=1, speedY=-1}) {texture=&T_Inky_Smoke_Tile_Chromatic}
  n6 = Panner(Coordinate=Multiply(A=@n2, B=TextureCoordinate() {uTiling=2, vTiling=2}), Time=@n4) {speedX=0.1, speedY=-1}
  n7 = Panner(Coordinate=Multiply(A=@n2, B=TextureCoordinate() {uTiling=6, vTiling=3}), Time=@n4) {speedX=-0.05, speedY=-1}
  n8 = Power(Base=Add(
    A=TextureSample(UVs=@n6) {texture=/Game/Global/Textures/Default/Masks_and_Gradients/T_BlasterBeamDistort}.G
    B=TextureSample(UVs=@n7) {texture=&T_BlasterBeamDistort}.G
   ))
  n9 = TextureCoordinate() {uTiling=0.1, vTiling=0.75}
  n10 = Panner(
   Coordinate=MaterialFunctionCall(XYZ=Multiply(A=@n2, B=@n9), XY=@n9) {materialFunction=&Swizzle}.YX
   Time=@n4
  ) {speedX=-1, speedY=1}
  n11 = Multiply(A=Power(Base=TextureSample(UVs=@n10) {texture=&T_BlasterBeamDistort}.G) {constExponent=4}) {constB=6}
  n12 = Clamp(None=Add(A=Power(Base=@n5.RGB), B=Add(A=@n8, B=@n11)))
  n13 = MaterialFunctionCall() {materialFunction=&LinearGradient}
  n14 = OneMinus(None=@n13.VGradient)
  n15 = Clamp(None=Multiply(A=Multiply(A=@n14, B=@n13.VGradient)) {constB=6})
  n16 = Transform(None=Constant3Vector() {constant=<rgba 1 0 0 0>}) {transformSourceType=TRANSFORMSOURCE_ParticleWorld}
  n17 = ScalarParameter() {defaultValue=3, parameterName=FresnelExp}
  n18 = Multiply(
   A=@n12
   B=Multiply(
    A=@n15
    B=MaterialFunctionCall(Axis=@n16, Exponent=@n17) {materialFunction=&AxisAlignedFresnel}.Fresnel
   )
  )
  n19 = Multiply(
   A=DepthFade(Opacity=Multiply(A=@n1.A, B=@n18), FadeDistance=ScalarParameter() {defaultValue=64, parameterName=Depth Fade Distance}) {fadeDistanceDefault=64}
   B=MaterialFunctionCall(Distance=ScalarParameter() {defaultValue=500, parameterName=Distance Fade}) {materialFunction=&Near_Camera_Fade}.Result
  )
  n20 = Multiply(A=@n1.A, B=@n15)
  FeatureLevelSwitch(Default=@n19, ES2=@n20, ES3_1=@n20, SM4_REMOVED=@n19, SM5=@n19, SM6=@n19)
  n21 = Multiply(
   A=LinearInterpolate(
    A=VectorParameter() {parameterName=ColorB, defaultValue=<rgba 0.0478 0.0048 2 1>}.RGB
    B=VectorParameter() {parameterName=ColorA, defaultValue=<rgba 1.1093 0.1179 3 1>}.RGB
    Alpha=@n12
   )
   B=@n1.RGB
  )
  n22 = MaterialFunctionCall(LDR Input=Clamp(None=@n21), HDR Tint=@n21, Desaturate Luminance=ScalarParameter() {defaultValue=1.1, parameterName=Desaturate Luminance}) {materialFunction=&DeriveHDRfromLDR}
  n23 = Multiply(A=@n22.RemappedRange) {constB=0.1}
  n24 = Multiply(A=@n22.RemappedRange)
  MaterialFunctionCall(Emissive=StaticSwitchParameter(
    True=MaterialFunctionCall(Color_Morning=@n23, Color_Day=@n24, Color_Evening=@n24, Color_Night=@n23) {materialFunction=&Color_Time_of_Day}.Result
    False=@n22.RemappedRange
   ) {parameterName=Color_Time_of_Day}) {materialFunction=&MF_FX_EmissiveAdjustments}
  n25 = MaterialFunctionCall(XY=TextureCoordinate()) {materialFunction=&Swizzle}
  n26 = StaticSwitchParameter(
   True=TextureSample(UVs=Panner(Coordinate=@n25.YX, Time=@n4) {speedX=-1, speedY=1}) {samplerType=SAMPLERTYPE_Masks, texture=/Engine/Functions/Engine_MaterialFunctions02/ExampleContent/Textures/SineWaveHeightMap}.R
   False=@n11
  ) {parameterName=Smoth Sine Wave WPO}
  n27 = Multiply(A=Multiply(A=Multiply(A=@n15, B=@n26), B=VertexNormalWS()), B=ScalarParameter() {defaultValue=20, parameterName=WPOAmount})
  n28 = VectorParameter() {parameterName=EndPoint, defaultValue=<rgba 67471.75 -43466.2109 3683.8064 0>}
  n29 = WorldPosition()
  n30 = ComponentMask(None=TextureCoordinate()) {g=true}
  n31 = ScalarParameter() {defaultValue=1, parameterName=Contrast}
  n32 = Multiply(A=Multiply(A=Subtract(A=@n28.RGB, B=@n29.XYZ), B=Power(Base=@n30, Exp=@n31)), B=ScalarParameter() {parameterName=Grace})
  Reroute(None=StaticSwitchParameter(True=Add(A=@n27, B=@n32), False=@n27) {parameterName=Bend it})
 ]
}
